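# Behavior tree for the FirePetal Wandering Boss's bombardment attack, built on its own Bombardment blackboard.
# A root Sequence, with a tactical-action service attached, first triggers the ranged Void Warp ability if the target actor is too close, then triggers the Bombardment ability.
# Force Success on the warp step lets bombardment run even when the warp is skipped or fails.
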BehaviorTree /FirePetal_WanderingBoss/BTInjection/BT_Bombardment_FirePetal_WanderingBoss
blackboard=&/BB_Bombardment_FirePetal_WanderingBoss
root {
 name=Sequence
 class=BTComposite_Sequence
 children [
  Trigger Local Warp Ability {
   class=FortBTTask_ExecuteGameplayAbility
   decorators [
    Too close to target? {
     class=FortBTDecorator_DistanceBetween
     properties {
      flowAbortMode=None
      nodeName=Too close to target?
      operator=LessOrEqual
      blackboardKeyB {selectedKeyName=AIEvaluator_Global_TargetActor}
      specifiedDistance {defaultValue=1500}
     }
    }
    Force Success {class=BTDecorator_ForceSuccess}
   ]
   properties {
    nodeName=Trigger Local Warp Ability
    gameplayAbilityTagContainer {defaultValue {gameplayTags [{tagName=Ability.WanderingBoss.VoidWarp.Ranged}]}}
   }
  }
  Trigger Bombardment Ability {
   class=FortBTTask_ExecuteGameplayAbility
   decorators []
   properties {
    nodeName=Trigger Bombardment Ability
    gameplayAbilityTagContainer {defaultValue {gameplayTags [{tagName=Ability.WanderingBoss.Bombardment}]}}
   }
  }
 ]
 decorators []
 properties {
  children [
   {
    childTask=&self:FortBTTask_ExecuteGameplayAbility_1
    decorators [
     &self:FortBTDecorator_DistanceBetween_0
     &self:BTDecorator_ForceSuccess_0
    ]
   }
   {childTask=&self:FortBTTask_ExecuteGameplayAbility_0}
  ]
  services [&self:FortAthenaBTService_SetInTacticalAction_0]
 }
 services [FortAthenaBTService_SetInTacticalAction_0 {
   class=FortAthenaBTService_SetInTacticalAction
   properties {shouldSetInTacticalAction {key=None}, tacticalActionIdentifierTag {defaultValue {_structType=&GameplayTag}}}
  }]
}
