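# Behavior tree injected for the FirePetal Wandering Boss that moves it around its leash area.
# When the target actor is over 4000 units away, a distance decorator aborts lower-priority wandering and triggers the ranged Void Warp ability.
# Otherwise, an EQS query writes a top-5% random destination, with blackboard-bindable min/max distances, and a wobbly bot move follows.
# Force Success on the query lets the move run even when no location is found.
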BehaviorTree /FirePetal_WanderingBoss/BTInjection/BT_MoveAround_LeashArea_WanderingBoss
blackboard=&/BB_MoveAround_LeashArea_WanderingBoss
root {
 name=Selector
 class=BTComposite_Selector
 children [
  Selector {
   class=BTComposite_Selector
   children [Trigger Local Warp Ability {
     class=FortBTTask_ExecuteGameplayAbility
     decorators []
     properties {
      nodeName=Trigger Local Warp Ability
      gameplayAbilityTagContainer {defaultValue {gameplayTags [{tagName=Ability.WanderingBoss.VoidWarp.Ranged}]}}
     }
    }]
   decorators [Distance Between {
     class=FortBTDecorator_DistanceBetween
     properties {
      flowAbortMode=LowerPriority
      nodeName=Distance Between
      operator=Greater
      blackboardKeyB {selectedKeyName=AIEvaluator_Global_TargetActor}
      specifiedDistance {defaultValue=4000}
     }
    }]
   properties {
    children [{
      childComposite=null
      childTask=&self:FortBTTask_ExecuteGameplayAbility_1
      decorators []
     }]
    services []
   }
   services []
  }
  Sequence {
   class=BTComposite_Sequence
   children [
    Selector {
     class=BTComposite_Selector
     children [Find Wander Location {
       class=BTTask_RunEQSQuery
       properties {
        nodeName=Find Wander Location
        blackboardKey {selectedKeyName=AIEvaluator_Patrolling_Destination}
        eQSRequest {
         queryTemplate=/AICore/Humanoids/EQS/EQS_MoveAround_LeashArea
         runMode=RandomBest5Pct
         queryConfig [
          {
           bAllowBBKey=true
           paramName=MinDistanceSelf
           paramType=Float
           value=500
           bBKey {bNoneIsAllowedValue=true, selectedKeyName=MinDistanceFromSelf}
          }
          {
           bAllowBBKey=true
           paramName=MaxDistanceSelf
           paramType=Float
           value=2000
           bBKey {bNoneIsAllowedValue=true, selectedKeyName=MaxDistanceFromSelf}
          }
         ]
        }
       }
      }]
     decorators [Force Success {class=BTDecorator_ForceSuccess}]
     properties {
      children [{
        childComposite=null
        childTask=&self:BTTask_RunEQSQuery_0
        decorators []
       }]
      services []
     }
     services []
    }
    "Move  To Location" {
     class=FortAthenaBTTask_BotMoveTo
     decorators []
     properties {
      movementResultKeyName=None
      nodeName="Move  To Location"
      acceptableRadius {defaultValue=50}
      allowRandomWobble {defaultValue=true}
      bAllowStrafe {defaultValue=false}
      blackboardKey {selectedKeyName=AIEvaluator_Patrolling_Destination}
      observedBlackboardValueTolerance {defaultValue=200}
     }
    }
   ]
   decorators []
   properties {children [
     {
      childComposite=&self:BTComposite_Selector_0
      childTask=null
      decorators [&self:BTDecorator_ForceSuccess_0]
     }
     {childTask=&self:FortAthenaBTTask_BotMoveTo_1}
    ]}
  }
 ]
 decorators []
 properties {
  children [
   {
    childComposite=&self:BTComposite_Selector_2
    childTask=null
    decorators [&self:FortBTDecorator_DistanceBetween_2]
   }
   {
    childComposite=&self:BTComposite_Sequence_0
    childTask=null
    decorators []
   }
  ]
  services []
 }
 services []
}
