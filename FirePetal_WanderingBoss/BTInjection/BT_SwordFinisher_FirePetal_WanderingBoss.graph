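# Injected behavior tree that runs the Fire Petal wandering boss's sword combo finisher against the evaluator's target actor.
# A Sequence first triggers the melee Void Warp ability when the target is too far, with Force Success so a skipped or failed warp still continues.
# It then triggers the Finisher ability by gameplay tag if the target is within a wider range.
# A tactical-action service runs on the Sequence throughout.
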BehaviorTree /FirePetal_WanderingBoss/BTInjection/BT_SwordFinisher_FirePetal_WanderingBoss
blackboard=&/BB_SwordFinisher_FirePetal_WanderingBoss
root {
 name=Sequence
 class=BTComposite_Sequence
 children [
  Trigger Local Warp Ability {
   class=FortBTTask_ExecuteGameplayAbility
   decorators [
    Too far from target? {
     class=FortBTDecorator_DistanceBetween
     properties {
      flowAbortMode=None
      nodeName=Too far from target?
      operator=Greater
      blackboardKeyB {selectedKeyName=AIEvaluator_Global_TargetActor}
      specifiedDistance {defaultValue=1000}
     }
    }
    Force Success {class=BTDecorator_ForceSuccess}
   ]
   properties {
    nodeName=Trigger Local Warp Ability
    gameplayAbilityTagContainer {defaultValue {gameplayTags [{tagName=Ability.WanderingBoss.VoidWarp.Melee}]}}
   }
  }
  Trigger Finisher Ability {
   class=FortBTTask_ExecuteGameplayAbility
   decorators [Close enough to target? {
     class=FortBTDecorator_DistanceBetween
     properties {
      flowAbortMode=None
      nodeName=Close enough to target?
      operator=LessOrEqual
      blackboardKeyB {selectedKeyName=AIEvaluator_Global_TargetActor}
      specifiedDistance {defaultValue=2500}
     }
    }]
   properties {
    nodeName=Trigger Finisher Ability
    gameplayAbilityTagContainer {defaultValue {gameplayTags [{tagName=Ability.WanderingBoss.SwordCombo.Finisher}]}}
   }
  }
 ]
 decorators []
 properties {
  children [
   {
    childTask=&self:FortBTTask_ExecuteGameplayAbility_1
    decorators [
     &self:FortBTDecorator_DistanceBetween_0
     &self:BTDecorator_ForceSuccess_0
    ]
   }
   {
    childTask=&self:FortBTTask_ExecuteGameplayAbility_0
    decorators [&self:FortBTDecorator_DistanceBetween_2]
   }
  ]
  services [&self:FortAthenaBTService_SetInTacticalAction_2]
 }
 services [FortAthenaBTService_SetInTacticalAction_2 {
   class=FortAthenaBTService_SetInTacticalAction
   properties {shouldSetInTacticalAction {key=None}, tacticalActionIdentifierTag {defaultValue {_structType=&GameplayTag}}}
  }]
}
