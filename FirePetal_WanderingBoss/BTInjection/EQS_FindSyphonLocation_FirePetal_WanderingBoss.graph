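# EQS query that finds the Syphon location for the FirePetal wandering boss, kept with its BTInjection assets.
# A single CurrentLocation generator takes the Bombardment current leash center context as the candidate point.
# It projects that point vertically onto the navmesh using the AvoidObstacles nav filter.
# No tests are attached, so the query applies no scoring or filtering beyond that projection.
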EnvQuery /FirePetal_WanderingBoss/BTInjection/EQS_FindSyphonLocation_FirePetal_WanderingBoss
options [{
  generator {
   name=EnvQueryGenerator_CurrentLocation_0
   class=EnvQueryGenerator_CurrentLocation
   properties {
    bAutoSortTests=true
    bCanRunAsync=false
    envQueryResultNormalizationOption=Default
    navDataOverrideContext=null
    optionName=""
    queryContext=/FirePetal_WanderingBoss/AI/Context/EnvQueryContext_Bombardment_CurrentLeashCenter_FirePetal_WanderingBoss.EnvQueryContext_Bombardment_CurrentLeashCenter_FirePetal_WanderingBoss_C
    projectionData {
     bCanDisableTrace=true
     bCanProjectDown=true
     bCanTraceOnGeometry=true
     bCanTraceOnNavMesh=true
     bOnlyBlockingHits=true
     bTraceComplex=false
     extentX=0
     extentY=10
     extentZ=10
     navigationFilter=/Game/AI/NavFilters/NavFilter_AvoidObstacles.NavFilter_AvoidObstacles_C
     postProjectionVerticalOffset=0
     projectDown=1024
     projectUp=1024
     serializedChannel=ECC_Visibility
     traceChannel=TraceTypeQuery1
     traceMode=Navigation
     traceProfileName=None
     traceShape=Line
     versionNum=1
    }
   }
  }
  tests []
 }]
