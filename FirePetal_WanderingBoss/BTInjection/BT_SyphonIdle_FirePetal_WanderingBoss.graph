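# Injected behavior tree for the FirePetal Wandering Boss's Syphon idle, run alongside an Escalate evaluator service.
# A force-success selector triggers a VoidWarp ability when the off-main-island check passes, moves to an EQS-found syphon location while Unaware, or waits when alert.
# A Simple Parallel then executes the ability from blackboard key Ability_TagContainer, with an Escalate task as its background branch, gated on execution status.
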BehaviorTree /FirePetal_WanderingBoss/BTInjection/BT_SyphonIdle_FirePetal_WanderingBoss
blackboard=&/BB_SyphonIdle_FirePetal_WanderingBoss
root {
 name=Selector
 class=BTComposite_Selector
 children [Sequence {
   class=BTComposite_Sequence
   children [
    Selector {
     class=BTComposite_Selector
     children [
      Selector {
       class=BTComposite_Selector
       children [Trigger Warp to Syphon Idle loc {
         class=FortBTTask_ExecuteGameplayAbility
         decorators []
         properties {
          nodeName=Trigger Warp to Syphon Idle loc
          gameplayAbilityTagContainer {defaultValue {gameplayTags [{tagName=Ability.WanderingBoss.VoidWarp.SyphonIdle}]}}
         }
        }]
       decorators [BTDecorator_IsOffMainIsland_FirePetal_WanderingBoss_C_0 {
         class=BTDecorator_IsOffMainIsland_FirePetal_WanderingBoss_C
         properties {
          bCheckConditionOnlyBlackBoardChanges=false
          bFoundSensor=false
          bInverseCondition=false
          bShowPropertyDetails=true
          currentIslandSensor=null
          customDescription=""
          flowAbortMode=None
          nodeName=""
          wanderingBossPawn=null
          arena Island Sensor Array []
         }
        }]
       properties {
        children [{
          childComposite=null
          childTask=&self:FortBTTask_ExecuteGameplayAbility_0
          decorators []
         }]
        services []
       }
       services []
      }
      Sequence {
       class=BTComposite_Sequence
       children [
        Find Root Idle Location {
         class=BTTask_RunEQSQuery
         properties {
          nodeName=Find Root Idle Location
          blackboardKey {selectedKeyName=AIEvaluator_Patrolling_Destination}
          eQSRequest {queryTemplate=&/EQS_FindSyphonLocation_FirePetal_WanderingBoss, runMode=SingleResult}
         }
        }
        Bot Move To {
         class=FortAthenaBTTask_BotMoveTo
         decorators []
         properties {
          movementResultKeyName=None
          nodeName=Bot Move To
          acceptableRadius {defaultValue=5}
          allowRandomWobble {defaultValue=true}
          bAllowStrafe {defaultValue=true}
          blackboardKey {selectedKeyName=AIEvaluator_Patrolling_Destination}
          observedBlackboardValueTolerance {defaultValue=4.75}
         }
        }
       ]
       decorators [Blackboard Based Condition {
         class=BTDecorator_Blackboard
         properties {
          arithmeticOperation=Equal
          flowAbortMode=Self
          intValue=0
          nodeName=Blackboard Based Condition
          notifyObserver=ValueChange
          stringValue=Unaware
          blackboardKey {selectedKeyName=AIEvaluator_Global_AlertLevel}
         }
        }]
       properties {children [
         {childTask=&self:BTTask_RunEQSQuery_2}
         {childTask=&self:FortAthenaBTTask_BotMoveTo_0}
        ]}
      }
      Exit BT if no longer unaware {
       class=BTTask_Wait
       decorators [Blackboard Based Condition {
         class=BTDecorator_Blackboard
         properties {
          arithmeticOperation=Greater
          flowAbortMode=Self
          intValue=0
          nodeName=Blackboard Based Condition
          notifyObserver=ValueChange
          stringValue=Unaware
          blackboardKey {selectedKeyName=AIEvaluator_Global_AlertLevel}
         }
        }]
       properties {nodeName=Exit BT if no longer unaware, waitTime {defaultValue=5}}
      }
     ]
     decorators [Force Success {class=BTDecorator_ForceSuccess}]
     properties {
      children [
       {
        childComposite=&self:BTComposite_Selector_0
        childTask=null
        decorators [&self:BTDecorator_IsOffMainIsland_FirePetal_WanderingBoss_C_0]
       }
       {
        childComposite=&self:BTComposite_Sequence_0
        childTask=null
        decorators [&self:BTDecorator_Blackboard_0]
       }
       {
        childComposite=null
        childTask=&self:BTTask_Wait_0
        decorators [&self:BTDecorator_Blackboard_1]
       }
      ]
      services []
     }
     services []
    }
    Simple Parallel {
     class=BTComposite_SimpleParallel
     children [
      TriggerAbility {
       class=FortBTTask_ExecuteGameplayAbility
       decorators []
       properties {nodeName=TriggerAbility, gameplayAbilityTagContainer {key=Ability_TagContainer, defaultValue {gameplayTags []}}}
      }
      FortAthenaBTTask_Escalate_0 {
       class=FortAthenaBTTask_Escalate
       children []
       decorators [Escalate? {
         class=BTDecorator_Blackboard
         properties {
          arithmeticOperation=GreaterOrEqual
          flowAbortMode=None
          intValue=3
          nodeName=Escalate?
          stringValue=ExecutionPending
          blackboardKey {selectedKeyName=AIEvaluator_Escalate_ExecutionStatus}
         }
        }]
       properties {bIgnoreRestartSelf=false, nodeName=""}
       services []
      }
     ]
     properties {
      children [
       {childTask=&self:FortBTTask_ExecuteGameplayAbility_1}
       {
        childTask=&self:FortAthenaBTTask_Escalate_0
        decorators [&self:BTDecorator_Blackboard_2]
       }
      ]
      services []
     }
     services []
    }
   ]
   decorators []
   properties {children [
     {
      childComposite=&self:BTComposite_Selector_1
      childTask=null
      decorators [&self:BTDecorator_ForceSuccess_0]
     }
     {childComposite=&self:BTComposite_SimpleParallel_1, childTask=null}
    ]}
  }]
 decorators []
 properties {
  children [{
    childComposite=&self:BTComposite_Sequence_1
    childTask=null
    decorators []
   }]
  services [&self:FortAthenaBTService_AIEvaluator_0]
 }
 services [Evaluating... Escalate {
   class=FortAthenaBTService_AIEvaluator
   properties {interval=0.5, nodeName=Evaluating... Escalate, aIEvaluatorClass {defaultValue=/Script/FortniteAIServer.FortAthenaAIBotEvaluator_Escalate}}
  }]
}
