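# Behavior tree for the FirePetal wandering boss that triggers its local Void Warp arena-reset ability.
# A root selector holds a single task that executes the gameplay ability tagged Ability.WanderingBoss.VoidWarp.ArenaReset.
# It has no decorators or services and uses its own Void Warp arena-reset blackboard.
# Its BTInjection folder suggests it is injected into the boss's AI rather than run as the main tree.
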BehaviorTree /FirePetal_WanderingBoss/BTInjection/BT_VoidWarp_ArenaReset_FirePetal_WanderingBoss
blackboard=&/BB_VoidWarp_ArenaReset_FirePetal_WanderingBoss
root {
 name=Selector
 class=BTComposite_Selector
 children [Trigger Local Warp Ability {
   class=FortBTTask_ExecuteGameplayAbility
   decorators []
   properties {
    nodeName=Trigger Local Warp Ability
    gameplayAbilityTagContainer {defaultValue {gameplayTags [{tagName=Ability.WanderingBoss.VoidWarp.ArenaReset}]}}
   }
  }]
 decorators []
 properties {
  children [{
    childComposite=null
    childTask=&self:FortBTTask_ExecuteGameplayAbility_0
    decorators []
   }]
  services []
 }
 services []
}
