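; Behavior tree injection evaluator that decides whether the FirePetal wandering boss's SyphonIdle behavior may start.
; On enter, it caches the controlled pawn from the behavior tree component's owning controller.
; The starting-condition check passes only if that pawn is valid and its owned gameplay tags lack the Ability.WanderingBoss.SyphonIdle.Paused tag, so that tag acts as a pause switch.
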(event EventOnEnter
  (params (OwnerComp :type Behavior%20Tree%20Component%20Object%20Reference))
  (bind _ascontroller (Utilities|Casting|CastToController (Components|GetOwner OwnerComp)))
  (Variables|Default|SetAIPawn (Pawn|GetControlledPawn _ascontroller)))

(fn EvaluateStartingConditions
  (params (OwnerComp :type Behavior%20Tree%20Component%20Object%20Reference))
  (bind _aipawn (Variables|Default|GetAIPawn))
  (Utilities|IsValid _aipawn
    (:Is%20Valid
      (bind _returnvalue (GameplayTags|GetOwnedGameplayTags _aipawn :!class /Script/GameplayTags.GameplayTagAssetInterface))
      (return (not (GameplayTags|HasTag _returnvalue "(TagName=\"Ability.WanderingBoss.SyphonIdle.Paused\")" true))))))
